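# Base material function for the 3DHB tool shaders, outputting normal, diffuse with opacity, and world position offset.
# Diffuse spreads texture red into RGB; opacity combines switchable texture alpha, powered vertex alpha and facing-based fade.
# An optional vertex-red-masked boost corrects over-transparent side faces before tinting.
# A height texture displaces along vertex normals, scaled by mesh scale and optionally vertex red; a custom-depth comparison is left unconnected.
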MaterialFunction /Game/Tools/3DHB/Shader/Base/MF_3DHB_Base
graph {nodes [
  n1 = TextureSampleParameter2D() {group=Base, parameterName=Base_NormalTexture, samplerType=SAMPLERTYPE_Normal, texture=&FlatNormalBlue}
  FunctionOutput(None=@n1.RGB) {outputName=OriginNormal}
  n2 = TextureSample(Tex=TextureObjectParameter() {group=Base, parameterName=Base_DiffuseTexture})
  n3 = AppendVector(A=@n2.RGB, B=OneMinus(None=@n2.A))
  n4 = ScalarParameter() {group=Base, parameterName=Base_IsFrom3D}
  n5 = Constant() {r=1}
  n6 = MaterialFunctionCall(Float4=LinearInterpolate(
    A=LinearInterpolate(A=@n2.RGBA, B=@n3, Alpha=@n4)
    B=AppendVector(A=@n2.RGB, B=@n5)
    Alpha=ScalarParameter() {group=Base, parameterName=Base_DisableTextureAlpha}
   )) {materialFunction=&BreakOutFloat4Components}
  n7 = Constant() {r=1}
  n8 = VertexColor()
  n9 = VertexInterpolator(VS=@n8.A)
  n10 = OneMinus(None=Power(Base=OneMinus(None=@n9.PS), Exp=ScalarParameter() {group=Base, parameterName=Base_VertexColorAlpha_Power}))
  n11 = ScalarParameter() {group=Base, parameterName=Base_VertexColorAlpha_Enable}
  n12 = Constant() {r=1}
  n13 = Multiply(A=@n6.A, B=ShadingPathSwitch(Default=@n12, Deferred=@n12, Forward=@n12, Mobile=@n12))
  n14 = DotProduct(A=CameraVectorWS(), B=VertexNormalWS())
  n15 = Constant() {r=0.2}
  n16 = Saturate(None=OneMinus(None=Add(A=@n14, B=@n15)))
  n17 = Saturate(None=Add(A=OneMinus(None=@n16), B=ScalarParameter() {defaultValue=0.35, group=Base, parameterName=Base_BackFaceFade_Lowestopacity}))
  n18 = MaterialFunctionCall(X=@n6.R, Y=@n6.R, Z=@n6.R, A=Multiply(A=LinearInterpolate(A=@n7, B=@n10, Alpha=@n11), B=Multiply(A=@n13, B=@n17))) {materialFunction=&MakeFloat4}
  n19 = Multiply(A=@n14)
  n20 = VectorParameter() {
   group=Base
   parameterName=Base_FixSideFaceOverTrans
   channelNames {a=ViewAngleLimite, b=MaskOffset, g=MaskMultiply, r=Strength}
   defaultValue=<rgba 0 1 0 0.8>
  }
  n21 = Saturate(None=OneMinus(None=Add(A=@n19, B=@n20.ViewAngleLimite) {constB=0.8}))
  n22 = Saturate(None=Add(A=Multiply(A=@n8.R, B=@n20.MaskMultiply), B=@n20.MaskOffset))
  n23 = Saturate(None=Floor(None=Add(A=@n8.R) {constB=0.01}))
  n24 = Multiply(A=Multiply(A=@n22, B=@n22), B=Multiply(A=OneMinus(None=@n23), B=@n20.Strength))
  n25 = Multiply(A=@n18.Result, B=Add(A=Multiply(A=@n21, B=@n24)))
  n26 = VectorParameter() {group=Base, parameterName=Base_Color/OpacityMultiply, defaultValue=<rgba 1 1 1 1>}
  FunctionOutput(None=Multiply(A=LinearInterpolate(A=@n18.Result, B=@n25, Alpha=@n11), B=AppendVector(A=@n26.RGB, B=@n26.RGB))) {outputName=OriginDiffuse}
  n27 = TextureObjectParameter() {group=Base, parameterName=Base_WPOTexture}
  n28 = Max(A=Add(A=TextureSample(Tex=@n27).R, B=ScalarParameter() {defaultValue=-0.5, group=Base, parameterName=Base_WPOOffset})) {constB=0}
  n29 = ScalarParameter() {defaultValue=40, group=Base, parameterName=Base_WPOStrength}
  n30 = FunctionInput() {bUsePreviewValueAsDefault=true, inputName=MeshScaleIn, inputType=FunctionInput_Scalar, previewValue=<1 0 0 1>}
  n31 = Multiply(A=Multiply(A=Multiply(A=@n28, B=@n29), B=VertexNormalWS()), B=@n30)
  FunctionOutput(None=LinearInterpolate(A=@n31, B=Multiply(A=@n31, B=@n8.R), Alpha=ScalarParameter() {group=Base, parameterName=Base_WPOMaskFromVertexColorRChannel})) {outputName=OriginWPO}
  n32 = Constant() {r=1}
  If(
   A=PixelDepth().R
   B=MaterialFunctionCall(Float3=SceneTexture() {sceneTextureId=PPI_CustomDepth}.Color) {materialFunction=&BreakOutFloat3Components}.R
   "A > B"=Constant()
   "A == B"=@n32
   "A < B"=@n32
  ) {equalsThreshold=0.1}
 ]}
