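# Material function in the 3DHB base shader tools that outputs screen-space UVs scaled by the actor's distance from the camera.
# It shifts viewport UV to center on zero, multiplies it by camera-to-actor distance times a small fixed factor, then restores the center offset.
# Scaling pivots on screen center, so a texture sampled with these UVs appears smaller as the actor moves away.
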MaterialFunction /Game/Tools/3DHB/Shader/Base/MF_3DHB_Base_ScreenUVWithDistanceScale
graph {nodes [
  n1 = Add(A=ScreenPosition().ViewportUV) {constB=-0.5}
  n2 = Multiply(A=Distance(A=CameraPositionWS(), B=ActorPositionWS())) {constB=0.003}
  FunctionOutput(None=Add(A=Multiply(A=@n1, B=@n2)) {constB=0.5})
 ]}
